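# Translucent unlit material for the storm cloud in the player's Melancholy emote, usable on Niagara mesh particles.
# Panning cloud and noise textures form normals, dotted with sun direction to blend time-of-day light and shadow colors, scaled by custom Schlick phase scattering.
# Purple lightning color comes from a position map and particle color.
# Crevice-masked shading also feeds particle-alpha erosion, with camera and angle fades or optional dither, plus masked normal-direction displacement.
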Material /Game/Animation/Game/MainPlayer/Emotes/Melancholy/VFX/M_Cloud_Melancholy_Emote
graph {
 properties {
  bCastRayTracedShadows=false
  bUsedWithNiagaraMeshParticles=true
  bUsedWithSplineMeshes=true
  blendMode=BLEND_Translucent
  pixelDepthOffsetMode=PDOM_Legacy
  shadingModel=MSM_Unlit
  usedShadingModels=Unlit
 }
 nodes [
  n1 = WorldPosition()
  n2 = ObjectPositionWS()
  n3 = Divide(
   A=Subtract(
    A=TransformPosition(None=@n1.XYZ) {transformType=TRANSFORMPOSSOURCE_World}
    B=TransformPosition(None=@n2) {transformType=TRANSFORMPOSSOURCE_World}
   )
   B=ObjectLocalBounds().Extents
  )
  Add(A=@n3) {constB=0.3}
  Constant3Vector() {constant=<rgba 0 0 1 0>}
  DynamicParameter() {parameterIndex=1, paramNames [LightningSides, LightningCenter, LightningLower, ""]}
  n4 = CollectionParameter() {
   collection=&FortniteMaterialParameters
   parameterName=SunlightDirectionalVector
  }
  n5 = ComponentMask(None=@n4) {b=true, g=true, r=true}
  n6 = TextureCoordinate()
  n7 = Multiply(
   A=Multiply(
    A=DynamicParameter() {paramNames [PanningSpeedX, PanningSpeedY, TileX, TileY]}.TileX
    B=ScalarParameter() {defaultValue=1, parameterName=Tile_N_X}
   )
   B=ComponentMask(None=@n6) {r=true}
  )
  n8 = Multiply(
   A=ComponentMask(None=@n6) {g=true}
   B=Multiply(
    A=ScalarParameter() {defaultValue=1, parameterName=Tile_N_Y}
    B=DynamicParameter() {paramNames [PanningSpeedX, PanningSpeedY, TileX, TileY]}.TileY
   )
  )
  n9 = VectorParameter() {parameterName=Speed, defaultValue=<rgba 0.008 0.0125 0 0>}
  n10 = DynamicParameter() {paramNames [PanningSpeedX, PanningSpeedY, TileX, TileY]}
  n11 = Multiply(A=ComponentMask(None=@n9.RGB) {r=true}, B=@n10.PanningSpeedX) {constA=1}
  n12 = Multiply(A=ComponentMask(None=@n9.RGB) {g=true}, B=@n10.PanningSpeedY) {constA=1}
  n13 = TextureSample(UVs=Panner(Coordinate=AppendVector(A=@n7, B=@n8), Speed=AppendVector(A=@n11, B=@n12))) {samplerType=SAMPLERTYPE_LinearColor, texture=/Game/VisualThreatManager/StormVisuals/LargeClouds/T_StormCloudMobile}
  n14 = VertexNormalWS()
  n15 = Panner(Coordinate=Multiply(A=TextureCoordinate()) {constB=0.5}) {speedX=-0.06, speedY=-0.1}
  n16 = TextureSampleParameter2D() {
   parameterName=Mask
   samplerType=SAMPLERTYPE_LinearGrayscale
   texture=&/T_Emote_Melancholy_Cloud_Base_Mask
  }
  n17 = MaterialFunctionCall(
   Normal=LinearInterpolate(
    A=LinearInterpolate(A=@n13.RGB, B=@n14)
    B=@n13.RGB
    Alpha=TextureSampleParameter2D(UVs=@n15) {parameterName=Tex1, samplerType=SAMPLERTYPE_LinearColor, texture=/EpicBaseTextures/Noises/T_Perlin_Noise_D}.RGB
   )
   Flatness=LinearInterpolate(B=ScalarParameter() {defaultValue=0.1, parameterName=FlattenDetailNormal}, Alpha=Power(Base=@n16.RGB) {constExponent=3}) {constA=1}
  ) {materialFunction=&FlattenNormal}
  n18 = LinearInterpolate(A=@n17.Result, B=VertexNormalWS())
  n19 = ScalarParameter() {defaultValue=0.1, parameterName=LerpVertNormal}
  n20 = TextureSampleParameter2D() {
   parameterName=Dissolve
   samplerType=SAMPLERTYPE_LinearGrayscale
   texture=&/T_Emote_Melancholy_Cloud_Lightning_Crevices
  }
  n21 = DotProduct(
   A=Transform(None=Normalize(VectorInput=@n5)) {transformSourceType=TRANSFORMSOURCE_World, transformType=TRANSFORM_Tangent}
   B=Multiply(A=LinearInterpolate(A=@n17.Result, B=@n18, Alpha=@n19), B=LinearInterpolate(Alpha=@n20.RGB) {constA=0.5})
  )
  n22 = Clamp(None=Power(Base=ConstantBiasScale(None=@n21) {bias=0.5, scale=0.6}, Exp=ScalarParameter() {defaultValue=1, parameterName=SoftenShadow}))
  n23 = Multiply(A=@n22, B=@n20.RGB)
  n24 = ParticleColor()
  n25 = LinearInterpolate(Alpha=@n24.A) {constA=1, constB=-1}
  n26 = MaterialFunctionCall(Distance=ScalarParameter() {parameterName=CameraDistanceFade}) {materialFunction=&Near_Camera_Fade}
  n27 = MaterialFunctionCall("FadeSensitivity (optional)"=ScalarParameter() {defaultValue=0.3, parameterName=AngleFadeSensitivity}) {materialFunction=&FadeBasedOnViewAngle}
  LinearInterpolate(
   A=Multiply(A=Subtract(A=@n23, B=@n25), B=Multiply(A=@n27.Result, B=@n26.Result) {constA=1})
   B=MaterialFunctionCall(Alpha Threshold=Multiply(A=@n25) {constB=-1}) {materialFunction=&DitherTemporalAA_Fortnite}.Result
   Alpha=ScalarParameter() {parameterName=Use_Dither}
  )
  n28 = VectorParameter() {parameterName=ShadowColorDay, defaultValue=<rgba 0.1679 0.1885 0.2049 1>}
  n29 = VectorParameter() {parameterName=ShadowColorMorning, defaultValue=<rgba 0.1552 0.1846 0.191 1>}
  n30 = VectorParameter() {parameterName=ShadowColorEvening, defaultValue=<rgba 0.1181 0.0709 0.0828 1>}
  n31 = VectorParameter() {parameterName=ShadowColorNight, defaultValue=<rgba 0.0952 0.1258 0.1319 1>}
  n32 = MaterialFunctionCall(
   Color_Morning=VectorParameter() {parameterName=LightColorMorning, defaultValue=<rgba 0.8992 0.5725 0.9062 1>}.RGB
   Color_Day=VectorParameter() {parameterName=LightColorDay, defaultValue=<rgba 0.8177 0.945 1 1>}.RGB
   Color_Evening=VectorParameter() {parameterName=LightColorEvening, defaultValue=<rgba 0.7969 0.4746 0.303 1>}.RGB
   Color_Night=VectorParameter() {parameterName=LightColorNight, defaultValue=<rgba 0.2019 0.2816 0.316 1>}.RGB
  ) {materialFunction=&Color_Time_of_Day}
  n33 = LinearInterpolate(
   A=Multiply(
    A=MaterialFunctionCall(Color_Morning=@n28.RGB, Color_Day=@n29.RGB, Color_Evening=@n30.RGB, Color_Night=@n31.RGB) {materialFunction=&Color_Time_of_Day}.Result
    B=@n32.Result
   )
   B=@n32.Result
   Alpha=@n22
  )
  n34 = Multiply(
   A=Multiply(A=VectorParameter() {parameterName=Particle color tint, defaultValue=<rgba 0.48 0.5867 0.8 1>}.RGB, B=ParticleColor().RGB)
   B=@n33
  ) {constB=0.8}
  n35 = OneMinus(None=Fresnel() {baseReflectFraction=0, exponent=0.7})
  n36 = CollectionParameter() {
   collection=&FortniteMaterialParameters
   parameterName=SunlightDirectionalVector
  }
  MaterialExpressionDotProduct_0 = DotProduct(A=ComponentMask(None=@n36) {b=true, g=true, r=true}, B=CameraVectorWS())
  MaterialExpressionScalarParameter_7 = ScalarParameter() {defaultValue=0.4, parameterName=Anisotropy}
  n37 = Custom!(CosTheta=@MaterialExpressionDotProduct_0, k=@MaterialExpressionScalarParameter_7) {
   code="\tfloat Inner = (1 + k * CosTheta);\r\n\treturn (1 - k * k) / (4 * PI * Inner * Inner);\r\n\r\n"
   description=Schlick Phase
   inputs [
    {
     inputName=CosTheta
     input {expression=&self:MaterialExpressionDotProduct_0}
    }
    {
     inputName=k
     input {expression=&self:MaterialExpressionScalarParameter_7}
    }
   ]
  }
  n38 = ScalarParameter() {defaultValue=24, parameterName=ScaleCloudColor}
  n39 = Multiply(A=Multiply(A=LinearInterpolate(A=@n34, B=@n33, Alpha=@n35) {constAlpha=1}, B=Multiply(A=@n37, B=@n38)), B=@n24.RGB)
  n40 = TextureSampleParameter2D() {
   parameterName=TexLightning
   samplerType=SAMPLERTYPE_LinearGrayscale
   texture=&/T_Emote_Melancholy_Cloud_Lightning_Position_Map
  }
  n41 = ParticleColor()
  n42 = Multiply(
   A=LinearInterpolate(
    A=Constant3Vector() {constant=<rgba 0.075 0.03 0.3 1>}
    B=Constant3Vector() {constant=<rgba 0.65 0.5 5 1>}
    Alpha=Power(Base=@n40.R) {constExponent=3}
   )
   B=@n41.RGB
  )
  LinearInterpolate(A=@n39, B=@n42)
  MaterialFunctionCall(Emissive=@n39) {materialFunction=&MF_FX_EmissiveAdjustments}
  MaterialFunctionCall(Emissive=@n42) {materialFunction=&MF_FX_EmissiveAdjustments}
  Multiply(A=@n40.R)
  n43 = Power(Base=@n23) {constExponent=1}
  n44 = VertexNormalWS()
  n45 = TextureSampleParameter2D() {
   parameterName=Mask
   samplerType=SAMPLERTYPE_LinearGrayscale
   texture=&T_Emote_Melancholy_Cloud_Base_Mask
  }
  Multiply(
   A=Multiply(A=Multiply(A=@n43, B=@n44), B=ScalarParameter() {defaultValue=7, parameterName=WPO_Intensity}) {constB=7}
   B=Saturate(None=Power(Base=@n45.RGB) {constExponent=3})
  )
  n46 = Multiply(A=Power(Base=@n40.R) {constExponent=1}, B=@n41.A)
  n47 = OneMinus(None=Fresnel() {baseReflectFraction=0, exponent=1})
  Multiply(B=Saturate(None=Multiply(A=@n46, B=@n47))) {constA=1}
  Saturate(None=Multiply(A=ComponentMask(None=@n3) {b=true}) {constB=10})
  ScalarParameter() {defaultValue=0.5, parameterName=Specular}
  ScalarParameter() {defaultValue=0.5, parameterName=Roughness}
  ScalarParameter() {parameterName=Metallic}
  Transform(None=@n1.XYZ)
  VectorParameter() {parameterName=Base Color}
 ]
}
